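# Sound cue for the opening sound of the Troops back bling cosmetic in Battle Royale.
# A single wave player feeds BackBling_Troops_Open_01 straight to the output, with no randomization or modulation.
# It uses close-range pet attenuation, the reactive back bling audio property sheet and the back bling sound class, at slightly reduced volume.
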SoundCue /BRCosmetics/Sounds/Backblings/Troops/BackBling_Troops_Open_Cue
root=WavePlayer_0
nodes [{id=WavePlayer_0, properties {soundWaveAssetPtr=&/BackBling_Troops_Open_01}}]
properties {
 attenuationSettings=&Fort_Pets_Close_Athena_Att
 audioPropertiesSheet=/BRCosmetics/Sounds/Settings/PropertySheets/Reactive/APS_Cosmetic_Reactive_Backbling
 maxDistance=1800.0001
 soundClassObject=/Game/Sounds/FortSoundClassesAndModes/Fort_Backblings
 volumeMultiplier=0.8
}
