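# Sound cue for the page-turn sound of the Troops back bling cosmetic.
# A random node picks one of ten equally weighted Turn wave variations each time it plays.
# It routes through the back bling sound class and reactive back bling property sheet, with close-range pet attenuation and reduced volume.
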SoundCue /BRCosmetics/Sounds/Backblings/Troops/BackBling_Troops_PageTurn_Cue
root=Random_0
nodes [
 {
  id=Random_0
  children [WavePlayer_1, WavePlayer_2, WavePlayer_3, WavePlayer_4, WavePlayer_5, WavePlayer_6, WavePlayer_7, WavePlayer_8, WavePlayer_9, WavePlayer_10]
  properties {weights [1,1,1,1,1,1,1,1,1,1]}
 }
 {id=WavePlayer_1, properties {soundWaveAssetPtr=&/BackBling_Troops_Turn_05}}
 {id=WavePlayer_10, properties {soundWaveAssetPtr=&/BackBling_Troops_Turn_08}}
 {id=WavePlayer_2, properties {soundWaveAssetPtr=&/BackBling_Troops_Turn_09}}
 {id=WavePlayer_3, properties {soundWaveAssetPtr=&/BackBling_Troops_Turn_10}}
 {id=WavePlayer_4, properties {soundWaveAssetPtr=&/BackBling_Troops_Turn_01}}
 {id=WavePlayer_5, properties {soundWaveAssetPtr=&/BackBling_Troops_Turn_02}}
 {id=WavePlayer_6, properties {soundWaveAssetPtr=&/BackBling_Troops_Turn_03}}
 {id=WavePlayer_7, properties {soundWaveAssetPtr=&/BackBling_Troops_Turn_04}}
 {id=WavePlayer_8, properties {soundWaveAssetPtr=&/BackBling_Troops_Turn_06}}
 {id=WavePlayer_9, properties {soundWaveAssetPtr=&/BackBling_Troops_Turn_07}}
]
properties {
 attenuationSettings=&Fort_Pets_Close_Athena_Att
 audioPropertiesSheet=/BRCosmetics/Sounds/Settings/PropertySheets/Reactive/APS_Cosmetic_Reactive_Backbling
 maxDistance=1800.0001
 soundClassObject=/Game/Sounds/FortSoundClassesAndModes/Fort_Backblings
 volumeMultiplier=0.501
}
